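; Deimos rift actor for the DADBRO playlist: a damageable, auto-fire-targetable rift that expires after a set lifetime.
; The server sets a replicated random mesh yaw, applies a health-multiplier effect and runs a lifetime timer, restarted on damage at a throttled rate.
; On expiry it applies a lapse effect and destroys itself.
; Clients register aim assist and play spawn audio; cosmetic events only print debug text.
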(event EventBeginPlay
  ;[1 1 1] EVENT -> BeginPlay
  (|CallParentFunction)
  (Networking|SwitchHasAuthority
    (:Authority
      (CallFunction|SetLifeTimeTimer)
      (bind _output_get (Variables|Default|SetRandomYaw (Math|Random|RandomFloatinRange 0.0 3.6e+02)))
      (Transformation|SetRelativeRotation (Variables|Default|GetCubeMesh) (Math|Rotator|MakeRotator 0.0 0.0 _output_get) false false)
      (GameplayEffects|ApplyGameplayEffectToSelf :self (Ability|GetAbilitySystemComponent self) :GameplayEffectClass "/Game/Athena/Playlists/DADBRO/GE_DADBRO_RiftHealthMultiplier.GE_DADBRO_RiftHealthMultiplier_C" :Level 0.0))
    (:Remote
      (AimAssist|RegisterAimAssistTarget self (Variables|Default|GetAimAssistCapsule))
      (Audio|PushSoundMixModifier "/Game/Sounds/FortSoundClassesAndModes/Fort_BossFight_Deimos_Spawn_MixMod.Fort_BossFight_Deimos_Spawn_MixMod")
      (Audio|PlaySoundatLocation "/Game/Sounds/Fort_Monster_Deimos/Spawner/DeimosSpawnerScream_Cue.DeimosSpawnerScream_Cue" (Transformation|GetActorLocation)))))

(event Custom|SetLifeTimeTimer
  ;[1 1 1] CUSTOM EVENT -> SetLifeTimeTimer
  (Utilities|Time|ClearandInvalidateTimerbyHandle (Variables|Default|GetLifeTimeTimerHandle))
  (bind _returnvalue (Utilities|Time|SetTimerbyEvent (EventDispatchers|CreateEvent :!event Custom|LifeTimeExpired) (Variables|Default|GetLifeTime)))
  (Variables|Default|SetLifeTimeTimerHandle _returnvalue))

(event EventOnDamageServer
  (params (Damage :type Float%20%28single-precision%29) (DamageTags :type Gameplay%20Tag%20Container%20Structure%20%28by%20ref%29) (Momentum :type Vector :default "0, 0, 0") (HitInfo :type Hit%20Result%20Structure%20%28by%20ref%29) (InstigatedBy :type Controller%20Object%20Reference) (DamageCauser :type Actor%20Object%20Reference) (EffectContext :type Gameplay%20Effect%20Context%20Handle%20Structure))
  ;[1 1 1] EVENT -> OnDamageServer
  (if (> (Utilities|TimeSince (Variables|Default|GetTimeLastSetLifeTimeTimer)) (Variables|Default|GetThrottleTime))
    (Variables|Default|SetTimeLastSetLifeTimeTimer (Utilities|Time|GetTimeSeconds))
    (CallFunction|SetLifeTimeTimer)))

(event Custom|LifeTimeExpired
  ;[1 1 1] CUSTOM EVENT -> LifeTimeExpired
  (GameplayEffects|ApplyGameplayEffectToSelf :self (Ability|GetAbilitySystemComponent self) :GameplayEffectClass (Variables|GameplayEffects|GetGE_LifeTimeLapsed) :Level 0.0)
  (Actor|DestroyActor))

(event EventBP_Cosmetic_Intro
  ;[0.5937 1 0.5123] Replicated Cosmetic Events
  ; Calling SetCosmeticState replicates to these.
  (|CallParentFunction)
  (Development|PrintString "BP_DeimosRift_Dynamic: Cosmetic Intro" false false))

(event EventBP_Cosmetic_Idle
  ;[0.5937 1 0.5123] Replicated Cosmetic Events
  ; Calling SetCosmeticState replicates to these.
  (|CallParentFunction)
  (Development|PrintString "BP_DeimosRift_Dynamic: Cosmetic Idle" false false))

(event EventBP_Cosmetic_RampUp
  ;[0.5937 1 0.5123] Replicated Cosmetic Events
  ; Calling SetCosmeticState replicates to these.
  (|CallParentFunction)
  (Development|PrintString "BP_DeimosRift_Dynamic: Cosmetic Ramp Up" false false))

(event EventBP_Cosmetic_ShouldDie
  ;[0.5937 1 0.5123] Replicated Cosmetic Events
  ; Calling SetCosmeticState replicates to these.
  (|CallParentFunction)
  (Development|PrintString "BP_DeimosRift_Dynamic: Cosmetic Should Die" false false))

(event EventEndPlay
  (params (EndPlayReason :type EEndPlayReason%20Enum :default "Destroyed"))
  (Networking|SwitchHasAuthority
    (:Remote
      (AimAssist|UnregisterAimAssistTarget self))))

(fn GetTargetOwner
  (return 0))

(fn IsValidAutoFireTarget
  (params (Attacker :type Actor%20Object%20Reference))
  (return true))

(fn OnRep_RandomYaw
  (Transformation|SetRelativeRotation (Variables|Default|GetCubeMesh) (Math|Rotator|MakeRotator 0.0 0.0 (Variables|Default|GetRandomYaw)) false false))

(fn OnRep_ServerSpawnTime)

(fn ShouldAlliesAutoFireOnDeployedTarget
  (return false))

(fn ShouldAutoFireOnSelfDeployedTarget
  (return false))

(fn ShouldDie
  (params (Damage :type Float%20%28single-precision%29) (EventInstigator :type Controller%20Object%20Reference) (DamageCauser :type Actor%20Object%20Reference) (bIsTest :type Boolean) (EffectContext :type Gameplay%20Effect%20Context%20Handle%20Structure%20%28by%20ref%29))
  (return true))

(fn ConstructionScript
  (bind _p_deimos_riftspawn_damagestate (Variables|Default|GetP_Deimos_RiftSpawn_DamageState))
  (|CallParentFunction)
  (Effects|Components|ParticleSystem|SetVectorParameter _p_deimos_riftspawn_damagestate "SizeScale" "0.700000,0.700000,0.700000")
  (Effects|Components|ParticleSystem|SetFloatParameter _p_deimos_riftspawn_damagestate "Radius" 0.0)
  (Components|DestroyComponent (Variables|Default|GetSM_CubeRunes)))
